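# Editor utility widget for previewing Alien Trooper head variants and batch-rendering face ROM delta movies.
# Update Head sets chin and forehead parts on two mannequins, then overrides human textures, patterns and eye color through dynamic material instances.
# Render starts a Tick-driven queue that passes each CID to Python `TechAnim.faceDeltaRender` whenever no movie is rendering.
# Update Textures looks unfinished, with a hard-coded Brain normal assigned to diffuse.
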EditorUtilityWidgetBlueprint /Game/NeverCook/TechArt/AT_FaceROMCustomData/FaceDeltaRenderWidget_ATVariant
graphs [
 EventGraph {nodes [
   "On Clicked (Render_Button)"() {is=title K2Node_ComponentBoundEvent}
   > Set IsRunning(IsRunning:bool=true) -> (Output_Get:bool)
   > Set CIDArrayIndex(CIDArrayIndex:int=0) -> (Output_Get:int=0)
   "On Clicked (UpdateHeadButton)"() {is=title K2Node_ComponentBoundEvent}
   > n1 = Sequence()
   n2 = "Get (a copy)"(
    Array:Array of Custom Character Part Object References=Get ChinCP() -> (ChinCP:Array of Custom Character Part Object References)
    Dimension 1="To Integer (Byte)"!(InByte:Byte=Get ChinType() -> (ChinType:HeadPartType Enum=NewEnumerator0)) -> (ReturnValue:int=0)
   ) -> (Output:Custom Character Part Object Reference)
   n3 = Get ManniquinOne() -> (ManniquinOne:Fort Player Mannequin Object Reference)
   > SetPart!(
    execute
    self:Fort Player Mannequin Object Reference=@n3
    Part:EFortCustomPartType Enum=Head
    PartAsset:Custom Character Part Object Reference=@n2
   ) -> (then, ReturnValue:bool)
   n4 = Get ManniquinTwo() -> (ManniquinTwo:Fort Player Mannequin Object Reference)
   > SetPart(self=@n4, Part=Head, PartAsset=@n2)
   > n5 = Sequence!(execute) -> (then_0, then_1, then_2:Exec)
   n6 = Get ForheadType() -> (ForheadType:HeadPartType Enum=NewEnumerator0)
   > n7 = Switch on HeadPartType!(execute, Selection:HeadPartType Enum=@n6) -> (
    NewEnumerator0:Exec
    NewEnumerator1:Exec
    NewEnumerator2:Exec
    NewEnumerator3:Exec
    NewEnumerator4:Exec
    NewEnumerator5:Exec
    NewEnumerator6:Exec
   ) {is=title K2Node_SwitchEnum}
   n8 = GetPartSkeletalMeshComponent!(self:Fort Player Mannequin Object Reference=Get ManniquinOne(), Part:EFortCustomPartType Enum=Head) -> (ReturnValue)
   n9 = Cast To MaterialInstanceDynamic!(
    execute=@n7.NewEnumerator3
    Object="Get (a copy)"(Array:Array of Material Interface Object References=GetMaterials!(self:Mesh Component Object Reference=@n8) -> (ReturnValue:Array of Material Interface Object References)) -> (Output:Material Interface Object Reference)
   ) -> (then, CastFailed, AsMaterial Instance Dynamic:Material Instance Dynamic Object Reference) {is=title K2Node_DynamicCast}
   n10 = GetPartSkeletalMeshComponent(self=Get ManniquinTwo(), Part=Head)
   > n11 = Cast To MaterialInstanceDynamic(Object="Get (a copy)"(Array=GetMaterials(self=@n10)))
   > n12 = Sequence!(execute) -> (then_0, then_1, then_2, then_3:Exec, then_4:Exec)
   n13 = Get HumanOverrideTextures() -> (HumanOverrideTextures:Array of Texture Object References)
   > SetTextureParameterValue(
    self=[@n11, @n9]
    ParameterName=AT_Diffuse
    Value="Get (a copy)"(Array:Array of Texture Object References=@n13) -> (Output:Texture Object Reference)
   )
   SetTextureParameterValue(execute=@n12.then_1, self=[@n11, @n9], ParameterName=AT_FXMask, Value="Get (a copy)"(Array=@n13, Dimension 1=1))
   SetTextureParameterValue(execute=@n12.then_2, self=[@n11, @n9], ParameterName=Normals, Value="Get (a copy)"(Array=@n13, Dimension 1=2))
   SetTextureParameterValue(execute=@n12.then_3, self=[@n11, @n9], ParameterName=M, Value="Get (a copy)"(Array=@n13, Dimension 1=3))
   SetTextureParameterValue(execute=@n12.then_4, self=[@n11, @n9], ParameterName=SpecularMasks, Value="Get (a copy)"(Array=@n13, Dimension 1=4))
   Override Pattern!(exec=@n5.then_1, Override?:bool=Get OverridePatterns?() -> (OverridePatterns?:bool), Part:EFortCustomPartType Enum=Head) -> (return:Exec) {is=title K2Node_MacroInstance}
   Override Eye Color!(exec=@n5.then_2, Part:EFortCustomPartType Enum=Head) -> (return:Exec) {is=title K2Node_MacroInstance}
   n14 = "Get (a copy)"(
    Array:Array of Custom Character Part Object References=Get ForheadCP() -> (ForheadCP:Array of Custom Character Part Object References)
    Dimension 1="To Integer (Byte)"(InByte=@n6) -> (ReturnValue=0)
   ) -> (Output:Custom Character Part Object Reference)
   SetPart(execute=@n1.then_1, self=@n3, Part=Face, PartAsset=@n14)
   > SetPart(self=@n4, Part=Face, PartAsset=@n14)
   > n15 = Sequence!(execute) -> (then_0, then_1, then_2)
   > n16 = Switch on HeadPartType(Selection=@n6)
   n17 = GetPartSkeletalMeshComponent(self=Get ManniquinOne(), Part=Face)
   n18 = Cast To MaterialInstanceDynamic(
    execute=@n16.NewEnumerator3
    Object="Get (a copy)"(Array:Array of Material Interface Object References=GetMaterials(self=@n17)) -> (Output:Material Interface Object Reference)
   )
   n19 = GetPartSkeletalMeshComponent(self=Get ManniquinTwo(), Part=Face)
   > n20 = Cast To MaterialInstanceDynamic(Object="Get (a copy)"(Array=GetMaterials(self=@n19)))
   n21 = Sequence!(execute=[@n20.CastFailed, @n20.then]) -> (then_0, then_1, then_2, then_3:Exec, then_4:Exec)
   n22 = Get HumanOverrideTextures()
   > SetTextureParameterValue(
    self=[@n18, @n20]
    ParameterName=AT_Diffuse
    Value="Get (a copy)"(Array:Array of Texture Object References=@n22) -> (Output:Texture Object Reference)
   )
   SetTextureParameterValue(execute=@n21.then_1, self=[@n18, @n20], ParameterName=AT_FXMask, Value="Get (a copy)"(Array=@n22, Dimension 1=1))
   SetTextureParameterValue(execute=@n21.then_2, self=[@n18, @n20], ParameterName=Normals, Value="Get (a copy)"(Array=@n22, Dimension 1=2))
   SetTextureParameterValue(execute=@n21.then_3, self=[@n18, @n20], ParameterName=M, Value="Get (a copy)"(Array=@n22, Dimension 1=3))
   SetTextureParameterValue(execute=@n21.then_4, self=[@n18, @n20], ParameterName=SpecularMasks, Value="Get (a copy)"(Array=@n22, Dimension 1=4))
   PrintString(execute=@n18.CastFailed, InString="Error, No Mannequin One Selected")
   Override Pattern(exec=@n15.then_1, Override?=Get OverridePatterns?(), Part=Face)
   Override Eye Color(exec=@n15.then_2, Part=Face)
   n23 = "On Property Changed (Render_DetailsView)"!() -> (OutputDelegate, then, PropertyName:Name=None) {is=title K2Node_ComponentBoundEvent}
   > Branch(Condition="Equal (Name)"!(A:Name=@n23.PropertyName, B:Name=PreviewCID) -> (ReturnValue:bool) {is=title K2Node_CallFunction})
   n24 = Get PreviewCID() -> (PreviewCID:Athena Character Item Definition Object Reference)
   > n25 = GetPathNameForLoadedAsset(LoadedAsset=@n24)
   > Execute Python Script!(
    execute
    PythonScript="import importlib\r\nimport unreal\r\nimport TechAnim.faceDeltaRender as fdr\r\nimportlib.reload(fdr)\r\nfdr.set_BP_FaceRom(cid_path=cid)"
    cid:String=@n25
   ) -> (then, ReturnValue) {inputs [cid]}
   "On Clicked (OpenLevel_Button)"() {is=title K2Node_ComponentBoundEvent}
   > Execute Python Script!(
    execute
    PythonScript="import importlib\r\nimport unreal\r\nimport TechAnim.faceDeltaRender as fdr\r\nimportlib.reload(fdr)\r\nfdr.open_FaceRom_level()"
   ) -> (then, ReturnValue)
   "On Clicked (UpdateTexturesButton_1)"() {is=title K2Node_ComponentBoundEvent}
   > Update Textures!(execute, self, Part:EFortCustomPartType Enum=Head) -> (then)
   > Update Textures(Part=Face)
   Event Construct()
   n26 = Self-Reference()
   > SetObject(self=Get Render_DetailsView() -> (Render_DetailsView:Details View Object Reference), NewObject=@n26)
   > SetObject(self=Get VariantDetailsView() -> (VariantDetailsView:Details View Object Reference), NewObject=@n26)
   Event Tick()
   > Branch(Condition=Get IsRunning() -> (IsRunning:bool))
   > n27 = IsRenderingMovie!(execute) -> (then, ReturnValue:bool)
   > n28 = Branch(Condition=@n27)
   n29 = "Get (a copy)"(
    Array:Array of Athena Character Item Definition Object References=Get CIDs() -> (CIDs:Array of Athena Character Item Definition Object References)
    Dimension 1=Get CIDArrayIndex() -> (CIDArrayIndex:int=0)
   ) -> (Output:Athena Character Item Definition Object Reference)
   n30 = GetPathNameForLoadedAsset(execute=@n28.else, LoadedAsset=@n29)
   n31 = Break Capture Resolution!(CaptureResolution:"Capture Resolution Structure (by ref)"=Get resolution() -> (resolution:Capture Resolution Structure)) -> (ResX:int, ResY:int) {is=title K2Node_BreakStruct}
   n32 = Break Float Range!(FloatRange:"Float Range Structure (by ref)"=Get FrameRange() -> (FrameRange:Float Range Structure)) -> (LowerBound:Float Range Bound Structure, UpperBound:Float Range Bound Structure) {is=title K2Node_BreakStruct}
   > Execute Python Script!(
    execute
    PythonScript="import importlib\r\nimport unreal\r\nimport TechAnim.faceDeltaRender as fdr\r\nimportlib.reload(fdr)\r\nfdr.generateFaceDeltaMovie(cid_path=cid, outputDir=output_dir, filename=filename, res_x=res_x, res_y=res_y, sf=start_frame, ef=end_frame)"
    cid:String=@n30
    filename:String=GetDisplayName(Object=@n29)
    startFrame:float=Break Float Range Bound!(FloatRangeBound:"Float Range Bound Structure (by ref)"=@n32) -> (Type:ERangeBoundTypes Enum, Value:float=~) {is=title K2Node_BreakStruct}.Value
    endFrame:float=Break Float Range Bound(FloatRangeBound=@n32.UpperBound) -> (Value=~).Value
    res_x:int=@n31
    res_y:int=@n31.ResY
    outputDir:String=Break Directory Path!(DirectoryPath:"Directory Path Structure (by ref)"=Get OutputDirectory() -> (OutputDirectory:Directory Path Structure)) -> (Path:String) {is=title K2Node_BreakStruct}
   ) -> (then, ReturnValue) {inputs [cid,filename,startFrame,endFrame,res_x,res_y,outputDir]}
   > n33 = Increment Int(Value=Get CIDArrayIndex() -> (CIDArrayIndex=0))
   n34 = Get CIDs()
   > Branch(Condition="integer >= integer"!(A:int=@n33, B:int=Length(TargetArray:Array of Athena Character Item Definition Object References=@n34)) -> (ReturnValue:bool) {is=title K2Node_CallFunction})
   > Set IsRunning(IsRunning=false)
   Event PreConstruct()
   Set Filename(Filename:String=GetDisplayName(Object=@n24)) -> (Output_Get:String)
  ]}
 OverRideTexturesFromCID {nodes [
   PrintString(InString=Overriding CID, bPrintToScreen=false)
   n1 = Get DataTable() -> (DataTable:Data Table Object Reference)
   > n2 = GetDataTableRowNames(Table=@n1)
   n3 = Get CurrentName() -> (CurrentName:Name=None)
   n4 = "Get (a copy)"(
    Array:Array of Names=@n2
    Dimension 1=int * int(A:int=Get CurrentDataTableSection() -> (CurrentDataTableSection:int=0), B:int=7) -> (ReturnValue:int=0) {is=title K2Node_CommutativeAssociativeBinaryOperator}
   ) -> (Output:Name)
   > PrintString(InString=Append!(A="To String (Name)"(InName=@n3), B=:, C:String="To String (Name)"(InName=@n4)) -> (ReturnValue), bPrintToScreen=false)
   > n5 = Get Data Table Row!(execute:Exec, DataTable:Data Table Object Reference=@n1, RowName:Name=@n4) -> (then:Exec, RowNotFound:Exec, ReturnValue:Alient Trooper Texture Structure Structure)
   > n6 = Sequence!(execute) -> (then_0, then_1, then_2, then_3, then_4, then_5:Exec)
   n7 = Break Alient Trooper Texture Structure!(AlientTrooperTextureStructure:"Alient Trooper Texture Structure Structure (by ref)"=@n5) -> (
    AT_Diffuse_7_#1:Texture 2D Object Reference
    AT_FXMask_8_#2:Texture 2D Object Reference
    Normal_11_#3:Texture 2D Object Reference
   ) {is=title K2Node_BreakStruct}
   n8 = Get TextureReturnSet() -> (TextureReturnSet:Array of Texture Object References)
   > Add(TargetArray:Array of Texture Object References=@n8, NewItem:"Texture Object Reference (by ref)"=@n7) {is=title K2Node_CallArrayFunction}
   Add(execute=@n6.then_1, TargetArray=@n8, NewItem=@n7.AT_FXMask_8_#2)
   Add(execute=@n6.then_2, TargetArray=@n8, NewItem=@n7.Normal_11_#3)
   n9 = Get CurrentDataTableSection() -> (CurrentDataTableSection=0)
   Add(
    execute=@n6.then_3
    TargetArray=@n8
    NewItem="Get (a copy)"(Array:Array of Texture Object References=Get M() -> (M:Array of Texture Object References), Dimension 1=@n9) -> (Output:Texture Object Reference)
   )
   Add(
    execute=@n6.then_4
    TargetArray=@n8
    NewItem="Get (a copy)"(Array=Get SpecularMasks() -> (SpecularMasks:Array of Texture Object References), Dimension 1=@n9)
   )
   Return Node!(execute=@n6.then_5, CurrentTextureSet:Array of Texture Object References=Get TextureReturnSet())
   PrintString(execute=@n5.RowNotFound, InString=ERROR WITH DATA TABLE IN OVERRIDE TEXTURES FUNCTION)
   OverRideTexturesFromCID() {is=title K2Node_FunctionEntry}
   Get CurrentVariantInt() -> (CurrentVariantInt:int=0)
  ]}
 OverrideEyeColor {nodes [
   n1 = Inputs!() -> (exec:Exec, Part:EFortCustomPartType Enum=Head)
   n2 = GetMaterials(self=GetPartSkeletalMeshComponent(self=Get ManniquinOne(), Part=@n1))
   > n3 = Cast To MaterialInstanceDynamic(Object="Get (a copy)"(Array:Array of Material Interface Object References=@n2) -> (Output:Material Interface Object Reference))
   n4 = GetMaterials(self=GetPartSkeletalMeshComponent(self=Get ManniquinTwo(), Part=@n1))
   > n5 = Cast To MaterialInstanceDynamic(Object="Get (a copy)"(Array=@n4))
   n6 = Switch on HeadPartType(execute=[@n5.CastFailed, @n5.then], Selection=Get ForheadType() -> (ForheadType=NewEnumerator0))
   SetVectorParameterValue(
    execute=[@n9.then, @n6.NewEnumerator0, @n6.NewEnumerator1, @n6.NewEnumerator3, @n6.NewEnumerator4, @n6.NewEnumerator5, @n6.NewEnumerator6]
    self=[@n5, @n3]
    ParameterName=AT_EyePrimary_HSV
    Value=Get EyeColor() -> (EyeColor:Linear Color Structure)
   )
   > Outputs!(return:Exec)
   n7 = Cast To MaterialInstanceDynamic(execute=@n6.NewEnumerator2, Object="Get (a copy)"(Array=@n2, Dimension 1=1))
   > n8 = Cast To MaterialInstanceDynamic(Object="Get (a copy)"(Array=@n4, Dimension 1=1))
   n9 = SetVectorParameterValue(execute=[@n8.CastFailed, @n8.then], self=[@n7, @n8], ParameterName=AT_EyePrimary_HSV, Value=Get EyeColor())
   PrintString(execute=@n7.CastFailed, InString="Error, No Mannequin One Selected")
   PrintString(execute=@n3.CastFailed, InString="Error, No Mannequin One Selected")
  ]}
 OverridePattern {nodes [
   n1 = Inputs!() -> (exec, Override?:bool=~, Part=Head)
   > n2 = Branch(Condition=@n1)
   n3 = GetMaterials(self=GetPartSkeletalMeshComponent(self=Get ManniquinOne(), Part=@n1.Part))
   > n4 = Cast To MaterialInstanceDynamic(Object="Get (a copy)"(Array=@n3))
   n5 = GetMaterials(self=GetPartSkeletalMeshComponent(self=Get ManniquinTwo(), Part=@n1.Part))
   > n6 = Cast To MaterialInstanceDynamic(Object="Get (a copy)"(Array=@n5))
   n7 = Switch on HeadPatternOptions!(execute=[@n6.CastFailed, @n6.then], Selection:HeadPatternOptions Enum=Get Pattern() -> (Pattern:HeadPatternOptions Enum=NewEnumerator0)) -> (
    NewEnumerator0:Exec
    NewEnumerator1:Exec
    NewEnumerator2:Exec
    NewEnumerator3:Exec
    NewEnumerator4:Exec
    NewEnumerator5:Exec
    NewEnumerator6:Exec
   ) {is=title K2Node_SwitchEnum}
   n8 = Get HumanPatternVariants() -> (HumanPatternVariants:Array of Texture Object References)
   > Set SelectedPattern(SelectedPattern:Texture Object Reference="Get (a copy)"(Array:Array of Texture Object References=@n8, Dimension 1=6) -> (Output:Texture Object Reference)) -> (Output_Get:Texture Object Reference)
   n9 = Get HumanPatternVariants_Normals() -> (HumanPatternVariants_Normals:Array of Texture Object References)
   > n10 = Set SelectedPattern_Normal(SelectedPattern_Normal:Texture Object Reference="Get (a copy)"(Array=@n9, Dimension 1=6)) -> (Output_Get:Texture Object Reference)
   n11 = Switch on HeadPartType(
    execute=[@n17.then, @n19.then, @n20.then, @n18.then, @n21.then, @n10.then, @n16.then]
    Selection=Get ForheadType() -> (ForheadType=NewEnumerator0)
   )
   n12 = SetTextureParameterValue(
    execute=[@n11.NewEnumerator0, @n11.NewEnumerator1, @n11.NewEnumerator4, @n11.NewEnumerator5, @n11.NewEnumerator6]
    self=[@n6, @n4]
    ParameterName=AT_FXMask
    Value=Get SelectedPattern() -> (SelectedPattern:Texture Object Reference)
   )
   Outputs(return=[@n12.then, @n15.then, @n2.else])
   n13 = Cast To MaterialInstanceDynamic(
    execute=@n11.NewEnumerator2
    Object="Get (a copy)"(Array:Array of Material Interface Object References=@n3, Dimension 1=1) -> (Output:Material Interface Object Reference)
   )
   > n14 = Cast To MaterialInstanceDynamic(Object="Get (a copy)"(Array=@n5, Dimension 1=1))
   n15 = SetTextureParameterValue(execute=[@n14.CastFailed, @n14.then], self=[@n13, @n14], ParameterName=AT_FXMask, Value=Get SelectedPattern())
   PrintString(execute=@n13.CastFailed, InString="Error, No Mannequin One Selected")
   SetTextureParameterValue(execute=@n11.NewEnumerator3, self=[@n6, @n4], ParameterName=AT_FXMask, Value=Get SelectedPattern())
   > SetTextureParameterValue(self=[@n6, @n4], ParameterName=Normals, Value=Get SelectedPattern_Normal() -> (SelectedPattern_Normal:Texture Object Reference))
   Set SelectedPattern(
    execute=@n7.NewEnumerator1
    SelectedPattern="Get (a copy)"(Array:Array of Texture Object References=@n8, Dimension 1=5) -> (Output:Texture Object Reference)
   )
   > n16 = Set SelectedPattern_Normal(SelectedPattern_Normal="Get (a copy)"(Array=@n9, Dimension 1=5))
   Set SelectedPattern(execute=@n7.NewEnumerator2, SelectedPattern="Get (a copy)"(Array=@n8, Dimension 1=2))
   > n17 = Set SelectedPattern_Normal(SelectedPattern_Normal="Get (a copy)"(Array=@n9, Dimension 1=2))
   Set SelectedPattern(execute=@n7.NewEnumerator3, SelectedPattern="Get (a copy)"(Array=@n8))
   > n18 = Set SelectedPattern_Normal(SelectedPattern_Normal="Get (a copy)"(Array=@n9))
   Set SelectedPattern(execute=@n7.NewEnumerator4, SelectedPattern="Get (a copy)"(Array=@n8, Dimension 1=4))
   > n19 = Set SelectedPattern_Normal(SelectedPattern_Normal="Get (a copy)"(Array=@n9, Dimension 1=4))
   Set SelectedPattern(execute=@n7.NewEnumerator5, SelectedPattern="Get (a copy)"(Array=@n8, Dimension 1=1))
   > n20 = Set SelectedPattern_Normal(SelectedPattern_Normal="Get (a copy)"(Array=@n9, Dimension 1=1))
   Set SelectedPattern(execute=@n7.NewEnumerator6, SelectedPattern="Get (a copy)"(Array=@n8, Dimension 1=3))
   > n21 = Set SelectedPattern_Normal(SelectedPattern_Normal="Get (a copy)"(Array=@n9, Dimension 1=3))
   PrintString(execute=@n4.CastFailed, InString="Error, No Mannequin One Selected")
  ]}
 Update Textures {nodes [
   n1 = OverRideTexturesFromCID!(execute, self) -> (then, CurrentTextureSet:Array of Texture Object References) {is=title K2Node_CallFunction}
   > Set CurrentTextureSet(CurrentTextureSet:Array of Texture Object References=@n1) -> (Output_Get:Array of Texture Object References)
   > n2 = Sequence!(execute) -> (then_0, then_1, then_2)
   > n3 = Sequence!(execute) -> (then_0, then_1, then_2, then_3:Exec, then_4:Exec)
   > SetTextureParameterValue(ParameterName=AT_Diffuse, Value="Get (a copy)"(Array=Get CurrentTextureSet() -> (CurrentTextureSet:Array of Texture Object References)))
   SetTextureParameterValue(execute=@n3.then_1, ParameterName=AT_FXMask, Value="Get (a copy)"(Array=Get CurrentTextureSet(), Dimension 1=1))
   SetTextureParameterValue(execute=@n3.then_2, ParameterName=Normals, Value="Get (a copy)"(Array=Get CurrentTextureSet(), Dimension 1=2))
   SetTextureParameterValue(execute=@n3.then_3, ParameterName=M, Value="Get (a copy)"(Array=Get CurrentTextureSet(), Dimension 1=3))
   SetTextureParameterValue(execute=@n3.then_4, ParameterName=SpecularMasks, Value="Get (a copy)"(Array=Get CurrentTextureSet(), Dimension 1=4))
   n4 = Sequence!(execute=@n2.then_1) -> (then_0, then_1, then_2)
   n5 = Get CurrentEyeColor() -> (CurrentEyeColor:Byte=0)
   > SetVectorParameterValue(
    ParameterName=AT_EyeColor
    Value="Get (a copy)"(
     Array:Array of Linear Color Structures=Get Eyes_AT_EyeColor() -> (Eyes_AT_EyeColor:Array of Linear Color Structures)
     Dimension 1="To Integer (Byte)"(InByte=@n5) -> (ReturnValue=0)
    ) -> (Output:Linear Color Structure)
   )
   n6 = Get CurrentEyeColor() -> (CurrentEyeColor=0)
   SetVectorParameterValue(
    execute=@n4.then_1
    ParameterName=AT_EyeEmissiveColor
    Value="Get (a copy)"(
     Array=Get Eyes_AT_EyeEmissiveColor() -> (Eyes_AT_EyeEmissiveColor:Array of Linear Color Structures)
     Dimension 1="To Integer (Byte)"(InByte=@n6) -> (ReturnValue=0)
    )
   )
   n7 = Get CurrentEyeColor() -> (CurrentEyeColor=0)
   SetVectorParameterValue(
    execute=@n4.then_2
    ParameterName=AT_EyeColorFresnel
    Value="Get (a copy)"(
     Array=Get Eyes_AT_EyeColorFresnel() -> (Eyes_AT_EyeColorFresnel:Array of Linear Color Structures)
     Dimension 1="To Integer (Byte)"(InByte=@n7) -> (ReturnValue=0)
    )
   )
   n8 = Sequence!(execute=@n2.then_2) -> (then_0, then_1, then_2, then_3:Exec)
   n9 = Get CurrentSkinColor() -> (CurrentSkinColor:Byte=0)
   n10 = Get SkelMeshMatRef() -> (SkelMeshMatRef:Material Instance Dynamic Object Reference)
   > SetVectorParameterValue(
    self=@n10
    ParameterName=AT_Primary_HSV
    Value="Get (a copy)"(
     Array=Get Skin_AT_Primary_HSV() -> (Skin_AT_Primary_HSV:Array of Linear Color Structures)
     Dimension 1="To Integer (Byte)"(InByte=@n9) -> (ReturnValue=0)
    )
   )
   n11 = Get CurrentSkinColor() -> (CurrentSkinColor=0)
   SetVectorParameterValue(
    execute=@n8.then_1
    self=@n10
    ParameterName=AT_Secondary_HSV
    Value="Get (a copy)"(
     Array=Get Skin_AT_Secondary_HSV() -> (Skin_AT_Secondary_HSV:Array of Linear Color Structures)
     Dimension 1="To Integer (Byte)"(InByte=@n11) -> (ReturnValue=0)
    )
   )
   n12 = Get CurrentSkinColor() -> (CurrentSkinColor=0)
   SetVectorParameterValue(
    execute=@n8.then_2
    self=@n10
    ParameterName=AT_PatternColor
    Value="Get (a copy)"(
     Array=Get Skin_AT_PatternColor() -> (Skin_AT_PatternColor:Array of Linear Color Structures)
     Dimension 1="To Integer (Byte)"(InByte=@n12) -> (ReturnValue=0)
    )
   )
   n13 = Get CurrentSkinColor() -> (CurrentSkinColor=0)
   SetVectorParameterValue(
    execute=@n8.then_3
    self=@n10
    ParameterName=AT_EmissiveColor
    Value="Get (a copy)"(
     Array=Get Skin_AT_EmissiveColor() -> (Skin_AT_EmissiveColor:Array of Linear Color Structures)
     Dimension 1="To Integer (Byte)"(InByte=@n13) -> (ReturnValue=0)
    )
   )
   n14 = Cast To SkeletalMeshComponentBudgeted!(execute, Object:Wildcard) -> (then, CastFailed, AsSkeletal Mesh Component Budgeted:Skeletal Mesh Component Budgeted Object Reference) {is=title K2Node_DynamicCast}
   > PrintString(InString=GetDisplayName(), Duration=10)
   PrintString(execute=@n14.CastFailed, InString=BAAAAAAAAAAAAAAAAAAD)
   n15 = Update Textures!() -> (then, Part:EFortCustomPartType Enum=Head) {is=title K2Node_FunctionEntry}
   > n16 = Sequence!(execute) -> (then_0, then_1)
   n17 = GetPartSkeletalMeshComponent(self=Get ManniquinOne(), Part=@n15)
   > n18 = Cast To MaterialInstanceDynamic(Object="Get (a copy)"(Array:Array of Material Interface Object References=GetMaterials(self=@n17)) -> (Output:Material Interface Object Reference))
   > SetTextureParameterValue(
    self=@n18
    ParameterName=AT_Diffuse
    Value=/Game/Characters/Player/Male/Medium/Bodies/M_MED_Alien_Trooper/Textures/Brain/T_M_MED_AlienTrooper_Brain_Head_N
   )
   n19 = GetPartSkeletalMeshComponent(self=Get ManniquinTwo(), Part=@n15)
   n20 = Cast To MaterialInstanceDynamic(execute=@n16.then_1, Object="Get (a copy)"(Array=GetMaterials(self=@n19)))
   > SetTextureParameterValue(
    self=@n20
    ParameterName=AT_Diffuse
    Value=&T_M_MED_AlienTrooper_Brain_Head_N
   )
   CreateDynamicMaterialInstance!(
    execute
    self:Primitive Component Object Reference
    ElementIndex:int=0
    SourceMaterial:Material Interface Object Reference
    OptionalName:Name=None
   ) -> (then, ReturnValue:Material Instance Dynamic Object Reference) {is=title K2Node_CallFunction}
   "To String (Integer)"(InInt=Length(TargetArray:Array of Wildcards))
   CreateDynamicMaterialInstance(self=@n17, ElementIndex=0, OptionalName=None)
   SetMaterial(self=@n14, ElementIndex=0)
   Get CurrentVariantInt() -> (CurrentVariantInt=0)
   Get CurrentVariantInt() -> (CurrentVariantInt=0)
   Get CurrentVariantInt() -> (CurrentVariantInt=0)
   Get CurrentVariantInt() -> (CurrentVariantInt=0)
   Get CurrentVariantInt() -> (CurrentVariantInt=0)
  ]}
]
variables [
 BlankHeadPattern {type=Texture, default=&T_BlacksRGB}
 ChinCP {
  type="array<CustomCharacterPart>"
  default [
   /Game/NeverCook/TechArt/AT_FaceROMCustomData/Data/CustomChinCP/CP/Custom_Horn
   /Game/NeverCook/TechArt/AT_FaceROMCustomData/Data/CustomChinCP/CP/Custom_Brain
   /Game/NeverCook/TechArt/AT_FaceROMCustomData/Data/CustomChinCP/CP/Custom_Cyclops
   /Game/NeverCook/TechArt/AT_FaceROMCustomData/Data/CustomChinCP/CP/Custom_Human
   /Game/NeverCook/TechArt/AT_FaceROMCustomData/Data/CustomChinCP/CP/Custom_Owl
   /Game/NeverCook/TechArt/AT_FaceROMCustomData/Data/CustomChinCP/CP/Custom_Tenticle
   /Game/NeverCook/TechArt/AT_FaceROMCustomData/Data/CustomChinCP/CP/Custom_Shark
  ]
 }
 ChinType {type=HeadPartType, default=Horns}
 CIDArrayIndex {type=integer}
 CIDs {type="array<AthenaCharacterItemDefinition>", default []}
 CurrentTextureSet {type="array<Texture>", default []}
 EyeColor {type=LinearColor, default=<rgba 0.01 1.7 0.7 1>}
 Filename {type=string, default=CID_005_Athena_Commando_M_Default}
 ForheadCP {
  type="array<CustomCharacterPart>"
  default [
   /Game/NeverCook/TechArt/AT_FaceROMCustomData/Data/CustomForeheadCP/CP/Custom_Forehead_Horn
   /Game/NeverCook/TechArt/AT_FaceROMCustomData/Data/CustomForeheadCP/CP/Custom_Forehead_Brain
   /Game/NeverCook/TechArt/AT_FaceROMCustomData/Data/CustomForeheadCP/CP/Custom_Forehead_Cyclops
   /Game/NeverCook/TechArt/AT_FaceROMCustomData/Data/CustomForeheadCP/CP/Custom_Forehead_Human
   /Game/NeverCook/TechArt/AT_FaceROMCustomData/Data/CustomForeheadCP/CP/Custom_Forehead_Owl
   /Game/NeverCook/TechArt/AT_FaceROMCustomData/Data/CustomForeheadCP/CP/Custom_Forehead_Tenticle
   /Game/NeverCook/TechArt/AT_FaceROMCustomData/Data/CustomForeheadCP/CP/Custom_Forehead_Shark
  ]
 }
 ForheadType {type=HeadPartType, default=Horns}
 FrameRange {type=FloatRange, default {lowerBound {type=Exclusive, value=0}, upperBound {type=Exclusive, value=1960}}}
 HumanOverrideTextures {
  type="array<Texture>"
  default [
   /BRCosmetics/Characters/Player/Male/Medium/Bodies/M_MED_Alien_Trooper/Textures/Human/T_M_MED_Alien_Trooper_Human_Head_D
   /BRCosmetics/Characters/Player/Male/Medium/Bodies/M_MED_Alien_Trooper/Textures/Human/T_M_MED_Alien_Trooper_Human_Head_FX_A
   /BRCosmetics/Characters/Player/Male/Medium/Bodies/M_MED_Alien_Trooper/Textures/Human/T_M_MED_Alien_Trooper_Human_Head_N
   /BRCosmetics/Characters/Player/Male/Medium/Bodies/M_MED_Alien_Trooper/Textures/Human/T_M_MED_Alien_Trooper_Human_Head_M
   /BRCosmetics/Characters/Player/Male/Medium/Bodies/M_MED_Alien_Trooper/Textures/Human/T_M_MED_Alien_Trooper_Human_Head_S
  ]
 }
 HumanPatternVariants {
  type="array<Texture>"
  default [
   &T_M_MED_Alien_Trooper_Human_Head_FX_A
   /BRCosmetics/Characters/Player/Male/Medium/Bodies/M_MED_Alien_Trooper/Textures/Human/T_M_MED_Alien_Trooper_Human_Head_FX_B
   /BRCosmetics/Characters/Player/Male/Medium/Bodies/M_MED_Alien_Trooper/Textures/Human/T_M_MED_Alien_Trooper_Human_Head_FX_C
   /BRCosmetics/Characters/Player/Male/Medium/Bodies/M_MED_Alien_Trooper/Textures/Human/T_M_MED_Alien_Trooper_Human_Head_FX_D
   /BRCosmetics/Characters/Player/Male/Medium/Bodies/M_MED_Alien_Trooper/Textures/Human/T_M_MED_Alien_Trooper_Human_Head_FX_E
   /BRCosmetics/Characters/Player/Male/Medium/Bodies/M_MED_Alien_Trooper/Textures/Human/T_M_MED_Alien_Trooper_Human_Head_FX_F
   /BRCosmetics/Characters/Player/Male/Medium/Bodies/M_MED_Alien_Trooper/Textures/Human/T_M_MED_Alien_Trooper_Human_Head_FX_G
  ]
 }
 HumanPatternVariants_Normals {
  type="array<Texture>"
  default [
   &T_M_MED_Alien_Trooper_Human_Head_N
   &T_M_MED_Alien_Trooper_Human_Head_N
   /BRCosmetics/Characters/Player/Male/Medium/Bodies/M_MED_Alien_Trooper/Textures/Human/T_M_MED_Alien_Trooper_Human_Head_N_C
   &T_M_MED_Alien_Trooper_Human_Head_N
   /BRCosmetics/Characters/Player/Male/Medium/Bodies/M_MED_Alien_Trooper/Textures/Human/T_M_MED_Alien_Trooper_Human_Head_N_A_E
   /BRCosmetics/Characters/Player/Male/Medium/Bodies/M_MED_Alien_Trooper/Textures/Human/T_M_MED_Alien_Trooper_Human_Head_N_F
   /BRCosmetics/Characters/Player/Male/Medium/Bodies/M_MED_Alien_Trooper/Textures/Human/T_M_MED_Alien_Trooper_Human_Head_N_G
  ]
 }
 IsRunning {type=boolean}
 ManniquinOne {type=FortPlayerMannequin, default=null}
 ManniquinTwo {type=FortPlayerMannequin, default=null}
 NewVar_0 {type=number}
 OpenButton {
  type=ButtonStyle
  default {
   clickedSlateSound {resourceObject=null}
   disabled {
    bIsDynamicallyLoaded=false
    drawAs=NoDrawType
    imageType=NoImage
    mirroring=NoMirror
    resourceName=None
    resourceObject=null
    tiling=NoTile
    imageSize=<0 0>
    margin=<ltrb 0 0 0 0>
    outlineSettings {
     bUseBrushTransparency=false
     roundingType=HalfHeightRadius
     width=0
     color {colorUseRule=UseColor_Specified, specifiedColor=<rgba 0 0 0 0>}
     cornerRadii=<0 0 0 1>
    }
    tintColor {colorUseRule=UseColor_Specified, specifiedColor=<rgba 1 1 1 1>}
    uVRegion {bIsValid=false, max=<0 0>, min=<0 0>}
   }
   disabledForeground {colorUseRule=UseColor_Foreground, specifiedColor=<rgba 1 0 1 1>}
   hovered {
    bIsDynamicallyLoaded=false
    drawAs=Image
    imageType=NoImage
    mirroring=NoMirror
    resourceName=None
    resourceObject=null
    tiling=NoTile
    imageSize=<32 32>
    margin=<ltrb 0 0 0 0>
    outlineSettings {
     bUseBrushTransparency=false
     roundingType=HalfHeightRadius
     width=0
     color {colorUseRule=UseColor_Specified, specifiedColor=<rgba 0 0 0 0>}
     cornerRadii=<0 0 0 1>
    }
    tintColor {colorUseRule=UseColor_Specified, specifiedColor=<rgba 1 1 1 1>}
    uVRegion {bIsValid=false, max=<0 0>, min=<0 0>}
   }
   hoveredForeground {colorUseRule=UseColor_Foreground, specifiedColor=<rgba 1 0 1 1>}
   hoveredSlateSound {resourceObject=null}
   normal {
    bIsDynamicallyLoaded=false
    drawAs=Image
    imageType=NoImage
    mirroring=NoMirror
    resourceName=None
    resourceObject=null
    tiling=NoTile
    imageSize=<32 32>
    margin=<ltrb 0 0 0 0>
    outlineSettings {
     bUseBrushTransparency=false
     roundingType=HalfHeightRadius
     width=0
     color {colorUseRule=UseColor_Specified, specifiedColor=<rgba 0 0 0 0>}
     cornerRadii=<0 0 0 1>
    }
    tintColor {colorUseRule=UseColor_Specified, specifiedColor=<rgba 1 1 1 1>}
    uVRegion {bIsValid=false, max=<0 0>, min=<0 0>}
   }
   normalForeground {colorUseRule=UseColor_Foreground, specifiedColor=<rgba 1 0 1 1>}
   normalPadding=<ltrb 0 0 0 0>
   pressed {
    bIsDynamicallyLoaded=false
    drawAs=Image
    imageType=NoImage
    mirroring=NoMirror
    resourceName=None
    resourceObject=null
    tiling=NoTile
    imageSize=<32 32>
    margin=<ltrb 0 0 0 0>
    outlineSettings {
     bUseBrushTransparency=false
     roundingType=HalfHeightRadius
     width=0
     color {colorUseRule=UseColor_Specified, specifiedColor=<rgba 0 0 0 0>}
     cornerRadii=<0 0 0 1>
    }
    tintColor {colorUseRule=UseColor_Specified, specifiedColor=<rgba 1 1 1 1>}
    uVRegion {bIsValid=false, max=<0 0>, min=<0 0>}
   }
   pressedForeground {colorUseRule=UseColor_Foreground, specifiedColor=<rgba 1 0 1 1>}
   pressedPadding=<ltrb 0 0 0 0>
   pressedSlateSound {resourceObject=null}
  }
 }
 OutputDirectory {type=DirectoryPath, default {path=D:/FaceROMs}}
 OverridePatterns? {type=boolean}
 Pattern {type=HeadPatternOptions, default=None}
 PreviewCID {type=AthenaCharacterItemDefinition, default=/Game/Athena/Items/Cosmetics/Characters/CID_005_Athena_Commando_M_Default}
 resolution {type=CaptureResolution, default {resX=1920, resY=1080}}
 SelectedPattern {type=Texture, default=null}
 SelectedPattern_Normal {type=Texture, default=null}
 Skel Mesh Mat Ref {type=MaterialInstanceDynamic}
]
widgets {
 info {namedSlotCount=4, rootWidgetClass=&ScrollBox, widgetCount=18}
 widgets [
  ScrollBox_0 {
   widgetClassPath=&ScrollBox
   widget=ScrollBox_0
   properties {
    widgetBarStyle {
     thickness=8
     draggedThumbImage {
      bIsDynamicallyLoaded=false
      drawAs=RoundedBox
      imageType=NoImage
      mirroring=NoMirror
      resourceName=None
      resourceObject=null
      tiling=NoTile
      imageSize=<0 0>
      margin=<ltrb 0 0 0 0>
      outlineSettings {
       bUseBrushTransparency=false
       roundingType=FixedRadius
       width=0
       color {colorUseRule=UseColor_Specified, specifiedColor=<rgba 0 0 0 0>}
       cornerRadii=<4 4 4 4>
      }
      tintColor {colorUseRule=UseColor_Specified, specifiedColor=<rgba 0.2159 0.2159 0.2159 1>}
      uVRegion {bIsValid=false, max=<0 0>, min=<0 0>}
     }
     horizontalBackgroundImage {
      bIsDynamicallyLoaded=false
      drawAs=NoDrawType
      imageType=NoImage
      mirroring=NoMirror
      resourceName=None
      resourceObject=null
      tiling=NoTile
      imageSize=<0 0>
      margin=<ltrb 0 0 0 0>
      outlineSettings {
       bUseBrushTransparency=false
       roundingType=HalfHeightRadius
       width=0
       color {colorUseRule=UseColor_Specified, specifiedColor=<rgba 0 0 0 0>}
       cornerRadii=<0 0 0 0>
      }
      tintColor {colorUseRule=UseColor_Specified, specifiedColor=<rgba 1 1 1 1>}
      uVRegion {bIsValid=false, max=<0 0>, min=<0 0>}
     }
     horizontalBottomSlotImage {
      bIsDynamicallyLoaded=false
      drawAs=NoDrawType
      imageType=NoImage
      mirroring=NoMirror
      resourceName=None
      resourceObject=null
      tiling=NoTile
      imageSize=<0 0>
      margin=<ltrb 0 0 0 0>
      outlineSettings {
       bUseBrushTransparency=false
       roundingType=HalfHeightRadius
       width=0
       color {colorUseRule=UseColor_Specified, specifiedColor=<rgba 0 0 0 0>}
       cornerRadii=<0 0 0 0>
      }
      tintColor {colorUseRule=UseColor_Specified, specifiedColor=<rgba 1 1 1 1>}
      uVRegion {bIsValid=false, max=<0 0>, min=<0 0>}
     }
     horizontalTopSlotImage {
      bIsDynamicallyLoaded=false
      drawAs=NoDrawType
      imageType=NoImage
      mirroring=NoMirror
      resourceName=None
      resourceObject=null
      tiling=NoTile
      imageSize=<0 0>
      margin=<ltrb 0 0 0 0>
      outlineSettings {
       bUseBrushTransparency=false
       roundingType=HalfHeightRadius
       width=0
       color {colorUseRule=UseColor_Specified, specifiedColor=<rgba 0 0 0 0>}
       cornerRadii=<0 0 0 0>
      }
      tintColor {colorUseRule=UseColor_Specified, specifiedColor=<rgba 1 1 1 1>}
      uVRegion {bIsValid=false, max=<0 0>, min=<0 0>}
     }
     hoveredThumbImage {
      bIsDynamicallyLoaded=false
      drawAs=RoundedBox
      imageType=NoImage
      mirroring=NoMirror
      resourceName=None
      resourceObject=null
      tiling=NoTile
      imageSize=<0 0>
      margin=<ltrb 0 0 0 0>
      outlineSettings {
       bUseBrushTransparency=false
       roundingType=FixedRadius
       width=0
       color {colorUseRule=UseColor_Specified, specifiedColor=<rgba 0 0 0 0>}
       cornerRadii=<4 4 4 4>
      }
      tintColor {colorUseRule=UseColor_Specified, specifiedColor=<rgba 0.2159 0.2159 0.2159 1>}
      uVRegion {bIsValid=false, max=<0 0>, min=<0 0>}
     }
     normalThumbImage {
      bIsDynamicallyLoaded=false
      drawAs=RoundedBox
      imageType=NoImage
      mirroring=NoMirror
      resourceName=None
      resourceObject=null
      tiling=NoTile
      imageSize=<0 0>
      margin=<ltrb 0 0 0 0>
      outlineSettings {
       bUseBrushTransparency=false
       roundingType=FixedRadius
       width=0
       color {colorUseRule=UseColor_Specified, specifiedColor=<rgba 0 0 0 0>}
       cornerRadii=<4 4 4 4>
      }
      tintColor {colorUseRule=UseColor_Specified, specifiedColor=<rgba 0.0953 0.0953 0.0953 1>}
      uVRegion {bIsValid=false, max=<0 0>, min=<0 0>}
     }
     verticalBackgroundImage {
      bIsDynamicallyLoaded=false
      drawAs=NoDrawType
      imageType=NoImage
      mirroring=NoMirror
      resourceName=None
      resourceObject=null
      tiling=NoTile
      imageSize=<0 0>
      margin=<ltrb 0 0 0 0>
      outlineSettings {
       bUseBrushTransparency=false
       roundingType=HalfHeightRadius
       width=0
       color {colorUseRule=UseColor_Specified, specifiedColor=<rgba 0 0 0 0>}
       cornerRadii=<0 0 0 0>
      }
      tintColor {colorUseRule=UseColor_Specified, specifiedColor=<rgba 1 1 1 1>}
      uVRegion {bIsValid=false, max=<0 0>, min=<0 0>}
     }
     verticalBottomSlotImage {
      bIsDynamicallyLoaded=false
      drawAs=NoDrawType
      imageType=NoImage
      mirroring=NoMirror
      resourceName=None
      resourceObject=null
      tiling=NoTile
      imageSize=<0 0>
      margin=<ltrb 0 0 0 0>
      outlineSettings {
       bUseBrushTransparency=false
       roundingType=HalfHeightRadius
       width=0
       color {colorUseRule=UseColor_Specified, specifiedColor=<rgba 0 0 0 0>}
       cornerRadii=<0 0 0 0>
      }
      tintColor {colorUseRule=UseColor_Specified, specifiedColor=<rgba 1 1 1 1>}
      uVRegion {bIsValid=false, max=<0 0>, min=<0 0>}
     }
     verticalTopSlotImage {
      bIsDynamicallyLoaded=false
      drawAs=NoDrawType
      imageType=NoImage
      mirroring=NoMirror
      resourceName=None
      resourceObject=null
      tiling=NoTile
      imageSize=<0 0>
      margin=<ltrb 0 0 0 0>
      outlineSettings {
       bUseBrushTransparency=false
       roundingType=HalfHeightRadius
       width=0
       color {colorUseRule=UseColor_Specified, specifiedColor=<rgba 0 0 0 0>}
       cornerRadii=<0 0 0 0>
      }
      tintColor {colorUseRule=UseColor_Specified, specifiedColor=<rgba 1 1 1 1>}
      uVRegion {bIsValid=false, max=<0 0>, min=<0 0>}
     }
    }
    widgetStyle {
     barThickness=8
     bottomShadowBrush {
      bIsDynamicallyLoaded=false
      drawAs=Box
      imageType=FullColor
      mirroring=NoMirror
      resourceName=../../../Engine/Content/Slate/Common/ScrollBoxShadowBottom.png
      resourceObject=null
      tiling=NoTile
      imageSize=<16 8>
      margin=<ltrb 0.5 0 0.5 1>
      outlineSettings {
       bUseBrushTransparency=false
       roundingType=HalfHeightRadius
       width=0
       color {colorUseRule=UseColor_Specified, specifiedColor=<rgba 0 0 0 0>}
       cornerRadii=<0 0 0 0>
      }
      tintColor {colorUseRule=UseColor_Specified, specifiedColor=<rgba 1 1 1 1>}
      uVRegion {bIsValid=false, max=<0 0>, min=<0 0>}
     }
     horizontalScrolledContentPadding=<ltrb 0 0 1 0>
     leftShadowBrush {
      bIsDynamicallyLoaded=false
      drawAs=Box
      imageType=FullColor
      mirroring=NoMirror
      resourceName=../../../Engine/Content/Slate/Common/ScrollBoxShadowLeft.png
      resourceObject=null
      tiling=NoTile
      imageSize=<8 16>
      margin=<ltrb 1 0.5 0 0.5>
      outlineSettings {
       bUseBrushTransparency=false
       roundingType=HalfHeightRadius
       width=0
       color {colorUseRule=UseColor_Specified, specifiedColor=<rgba 0 0 0 0>}
       cornerRadii=<0 0 0 0>
      }
      tintColor {colorUseRule=UseColor_Specified, specifiedColor=<rgba 1 1 1 1>}
      uVRegion {bIsValid=false, max=<0 0>, min=<0 0>}
     }
     rightShadowBrush {
      bIsDynamicallyLoaded=false
      drawAs=Box
      imageType=FullColor
      mirroring=NoMirror
      resourceName=../../../Engine/Content/Slate/Common/ScrollBoxShadowRight.png
      resourceObject=null
      tiling=NoTile
      imageSize=<8 16>
      margin=<ltrb 0 0.5 1 0.5>
      outlineSettings {
       bUseBrushTransparency=false
       roundingType=HalfHeightRadius
       width=0
       color {colorUseRule=UseColor_Specified, specifiedColor=<rgba 0 0 0 0>}
       cornerRadii=<0 0 0 0>
      }
      tintColor {colorUseRule=UseColor_Specified, specifiedColor=<rgba 1 1 1 1>}
      uVRegion {bIsValid=false, max=<0 0>, min=<0 0>}
     }
     topShadowBrush {
      bIsDynamicallyLoaded=false
      drawAs=Box
      imageType=FullColor
      mirroring=NoMirror
      resourceName=../../../Engine/Content/Slate/Common/ScrollBoxShadowTop.png
      resourceObject=null
      tiling=NoTile
      imageSize=<16 8>
      margin=<ltrb 0.5 1 0.5 0>
      outlineSettings {
       bUseBrushTransparency=false
       roundingType=HalfHeightRadius
       width=0
       color {colorUseRule=UseColor_Specified, specifiedColor=<rgba 0 0 0 0>}
       cornerRadii=<0 0 0 0>
      }
      tintColor {colorUseRule=UseColor_Specified, specifiedColor=<rgba 1 1 1 1>}
      uVRegion {bIsValid=false, max=<0 0>, min=<0 0>}
     }
     verticalScrolledContentPadding=<ltrb 0 0 0 1>
    }
   }
  }
  openLevel_VerticalBox {widgetClassPath=&VerticalBox, parent=ScrollBox_0, widget=openLevel_VerticalBox, slot {class=ScrollBoxSlot}}
  openLevel_ExpandableArea {
   widgetClassPath=/Script/UMG.ExpandableArea
   bIsVariable=true
   parent=openLevel_VerticalBox
   widget=openLevel_ExpandableArea
   properties {
    bIsExpanded=true
    borderBrush {
     bIsDynamicallyLoaded=false
     drawAs=RoundedBox
     imageType=NoImage
     mirroring=NoMirror
     resourceName=None
     resourceObject=null
     tiling=NoTile
     imageSize=<0 0>
     margin=<ltrb 0 0 0 0>
     outlineSettings {
      bUseBrushTransparency=false
      roundingType=FixedRadius
      width=0
      color {colorUseRule=UseColor_Specified, specifiedColor=<rgba 0 0 0 0>}
      cornerRadii=<4 4 4 4>
     }
     tintColor {colorUseRule=UseColor_Specified, specifiedColor=<rgba 0.0176 0.0176 0.0176 1>}
     uVRegion {bIsValid=false, max=<0 0>, min=<0 0>}
    }
    style {
     rolloutAnimationSeconds=0.1
     collapsedImage {
      bIsDynamicallyLoaded=false
      drawAs=Image
      imageType=Vector
      mirroring=NoMirror
      resourceName=../../../Engine/Content/Slate/Starship/Common/caret-right.svg
      resourceObject=null
      tiling=NoTile
      imageSize=<16 16>
      margin=<ltrb 0 0 0 0>
      outlineSettings {
       bUseBrushTransparency=false
       roundingType=HalfHeightRadius
       width=0
       color {colorUseRule=UseColor_Specified, specifiedColor=<rgba 0 0 0 0>}
       cornerRadii=<0 0 0 0>
      }
      tintColor {colorUseRule=UseColor_Specified, specifiedColor=<rgba 0.5271 0.5271 0.5271 1>}
      uVRegion {bIsValid=false, max=<0 0>, min=<0 0>}
     }
     expandedImage {
      bIsDynamicallyLoaded=false
      drawAs=Image
      imageType=Vector
      mirroring=NoMirror
      resourceName=../../../Engine/Content/Slate/Starship/Common/caret-down.svg
      resourceObject=null
      tiling=NoTile
      imageSize=<16 16>
      margin=<ltrb 0 0 0 0>
      outlineSettings {
       bUseBrushTransparency=false
       roundingType=HalfHeightRadius
       width=0
       color {colorUseRule=UseColor_Specified, specifiedColor=<rgba 0 0 0 0>}
       cornerRadii=<0 0 0 0>
      }
      tintColor {colorUseRule=UseColor_Specified, specifiedColor=<rgba 0.5271 0.5271 0.5271 1>}
      uVRegion {bIsValid=false, max=<0 0>, min=<0 0>}
     }
    }
   }
   slot {class=VerticalBoxSlot}
  }
  openLevel_label {
   widgetClassPath=&TextBlock
   bIsVariable=true
   namedSlotHost=openLevel_ExpandableArea
   widget=openLevel_label
   properties {
    text=Open Level
    font {
     bForceMonospaced=false
     bMaterialIsStencil=false
     fontMaterial=null
     fontObject=&Roboto
     letterSpacing=0
     monospacedWidth=1
     size=10
     skewAmount=0
     typefaceFontName=Bold
     outlineSettings {
      bApplyOutlineToDropShadows=false
      bMiteredCorners=false
      bSeparateFillAlpha=false
      outlineMaterial=null
      outlineSize=0
      outlineColor=<rgba 0 0 0 1>
     }
    }
   }
  }
  OpenLevel_Button {
   widgetClassPath=&Button
   namedSlotHost=openLevel_ExpandableArea
   widget=OpenLevel_Button
   properties {widgetStyle {
     clickedSlateSound {resourceObject=null}
     disabled {
      bIsDynamicallyLoaded=false
      drawAs=RoundedBox
      imageType=NoImage
      mirroring=NoMirror
      resourceName=None
      resourceObject=null
      tiling=NoTile
      imageSize=<0 0>
      margin=<ltrb 0 0 0 0>
      outlineSettings {
       bUseBrushTransparency=false
       roundingType=FixedRadius
       width=1
       color {colorUseRule=UseColor_ColorTable, specifiedColor=<rgba 0 0 0 0>}
       cornerRadii=<4 4 4 4>
      }
      tintColor {colorUseRule=UseColor_Specified, specifiedColor=<rgba 0.0395 0.0395 0.0395 1>}
      uVRegion {bIsValid=false, max=<0 0>, min=<0 0>}
     }
     disabledForeground {colorUseRule=UseColor_ColorTable, specifiedColor=<rgba 0 0 0 0>}
     hovered {
      bIsDynamicallyLoaded=false
      drawAs=RoundedBox
      imageType=NoImage
      mirroring=NoMirror
      resourceName=None
      resourceObject=null
      tiling=NoTile
      imageSize=<32 32>
      margin=<ltrb 0 0 0 0>
      outlineSettings {
       bUseBrushTransparency=false
       roundingType=FixedRadius
       width=1
       color {colorUseRule=UseColor_ColorTable, specifiedColor=<rgba 0 0 0 0>}
       cornerRadii=<4 4 4 4>
      }
      tintColor {colorUseRule=UseColor_Specified, specifiedColor=<rgba 0.0953 0.0953 0.0953 1>}
      uVRegion {bIsValid=false, max=<0 0>, min=<0 0>}
     }
     hoveredForeground {colorUseRule=UseColor_ColorTable, specifiedColor=<rgba 0 0 0 0>}
     hoveredSlateSound {resourceObject=null}
     normal {
      bIsDynamicallyLoaded=false
      drawAs=RoundedBox
      imageType=NoImage
      mirroring=NoMirror
      resourceName=None
      resourceObject=null
      tiling=NoTile
      imageSize=<32 32>
      margin=<ltrb 0 0 0 0>
      outlineSettings {
       bUseBrushTransparency=false
       roundingType=FixedRadius
       width=1
       color {colorUseRule=UseColor_ColorTable, specifiedColor=<rgba 0 0 0 0>}
       cornerRadii=<4 4 4 4>
      }
      tintColor {colorUseRule=UseColor_Specified, specifiedColor=<rgba 0.0395 0.0395 0.0395 1>}
      uVRegion {bIsValid=false, max=<0 0>, min=<0 0>}
     }
     normalForeground {colorUseRule=UseColor_ColorTable, specifiedColor=<rgba 0 0 0 0>}
     normalPadding=<ltrb 12 1.5 12 1.5>
     pressed {
      bIsDynamicallyLoaded=false
      drawAs=RoundedBox
      imageType=NoImage
      mirroring=NoMirror
      resourceName=None
      resourceObject=null
      tiling=NoTile
      imageSize=<32 32>
      margin=<ltrb 0 0 0 0>
      outlineSettings {
       bUseBrushTransparency=false
       roundingType=FixedRadius
       width=1
       color {colorUseRule=UseColor_ColorTable, specifiedColor=<rgba 0 0 0 0>}
       cornerRadii=<4 4 4 4>
      }
      tintColor {colorUseRule=UseColor_Specified, specifiedColor=<rgba 0.0284 0.0284 0.0284 1>}
      uVRegion {bIsValid=false, max=<0 0>, min=<0 0>}
     }
     pressedForeground {colorUseRule=UseColor_ColorTable, specifiedColor=<rgba 0 0 0 0>}
     pressedPadding=<ltrb 12 2.5 12 0.5>
     pressedSlateSound {resourceObject=null}
    }}
  }
  OpenLevel_CommonText {
   widgetClassPath=&CommonTextBlock
   parent=OpenLevel_Button
   widget=OpenLevel_CommonText
   properties {
    text=Open FaceROM Level
    font {
     bForceMonospaced=false
     bMaterialIsStencil=false
     fontMaterial=null
     fontObject=&BurbankSmall
     letterSpacing=0
     monospacedWidth=1
     size=17
     skewAmount=0
     typefaceFontName=Medium
     outlineSettings {
      bApplyOutlineToDropShadows=false
      bMiteredCorners=false
      bSeparateFillAlpha=false
      outlineMaterial=null
      outlineSize=0
      outlineColor=<rgba 0 0 0 1>
     }
    }
    shadowOffset=<0 0>
   }
   slot {class=ButtonSlot, properties {padding=<ltrb 10 0 10 0>}}
  }
  VariantTool {
   widgetClassPath=&ExpandableArea
   bIsVariable=true
   parent=openLevel_VerticalBox
   widget=VariantTool
   properties {
    bIsExpanded=true
    borderBrush {
     bIsDynamicallyLoaded=false
     drawAs=RoundedBox
     imageType=NoImage
     mirroring=NoMirror
     resourceName=None
     resourceObject=null
     tiling=NoTile
     imageSize=<0 0>
     margin=<ltrb 0 0 0 0>
     outlineSettings {
      bUseBrushTransparency=false
      roundingType=FixedRadius
      width=0
      color {colorUseRule=UseColor_Specified, specifiedColor=<rgba 0 0 0 0>}
      cornerRadii=<4 4 4 4>
     }
     tintColor {colorUseRule=UseColor_Specified, specifiedColor=<rgba 0.0176 0.0176 0.0176 1>}
     uVRegion {bIsValid=false, max=<0 0>, min=<0 0>}
    }
    style {
     rolloutAnimationSeconds=0.1
     collapsedImage {
      bIsDynamicallyLoaded=false
      drawAs=Image
      imageType=Vector
      mirroring=NoMirror
      resourceName=../../../Engine/Content/Slate/Starship/Common/caret-right.svg
      resourceObject=null
      tiling=NoTile
      imageSize=<16 16>
      margin=<ltrb 0 0 0 0>
      outlineSettings {
       bUseBrushTransparency=false
       roundingType=HalfHeightRadius
       width=0
       color {colorUseRule=UseColor_Specified, specifiedColor=<rgba 0 0 0 0>}
       cornerRadii=<0 0 0 0>
      }
      tintColor {colorUseRule=UseColor_Specified, specifiedColor=<rgba 0.5271 0.5271 0.5271 1>}
      uVRegion {bIsValid=false, max=<0 0>, min=<0 0>}
     }
     expandedImage {
      bIsDynamicallyLoaded=false
      drawAs=Image
      imageType=Vector
      mirroring=NoMirror
      resourceName=../../../Engine/Content/Slate/Starship/Common/caret-down.svg
      resourceObject=null
      tiling=NoTile
      imageSize=<16 16>
      margin=<ltrb 0 0 0 0>
      outlineSettings {
       bUseBrushTransparency=false
       roundingType=HalfHeightRadius
       width=0
       color {colorUseRule=UseColor_Specified, specifiedColor=<rgba 0 0 0 0>}
       cornerRadii=<0 0 0 0>
      }
      tintColor {colorUseRule=UseColor_Specified, specifiedColor=<rgba 0.5271 0.5271 0.5271 1>}
      uVRegion {bIsValid=false, max=<0 0>, min=<0 0>}
     }
    }
   }
   slot {class=VerticalBoxSlot}
  }
  VariantToolHeader {
   widgetClassPath=&TextBlock
   bIsVariable=true
   namedSlotHost=VariantTool
   widget=VariantToolHeader
   properties {
    text=AlientTrooperVariants
    font {
     bForceMonospaced=false
     bMaterialIsStencil=false
     fontMaterial=null
     fontObject=&Roboto
     letterSpacing=0
     monospacedWidth=1
     size=10
     skewAmount=0
     typefaceFontName=Bold
     outlineSettings {
      bApplyOutlineToDropShadows=false
      bMiteredCorners=false
      bSeparateFillAlpha=false
      outlineMaterial=null
      outlineSize=0
      outlineColor=<rgba 0 0 0 1>
     }
    }
   }
  }
  VerticalBox_0 {widgetClassPath=&VerticalBox, namedSlotHost=VariantTool, widget=VerticalBox_0}
  VariantDetailsView {
   widgetClassPath=&DetailsView
   parent=VerticalBox_0
   widget=VariantDetailsView
   properties {categoriesToShow [ATVariant]}
   slot {class=VerticalBoxSlot}
  }
  UpdateHeadButton {
   widgetClassPath=&Button
   parent=VerticalBox_0
   widget=UpdateHeadButton
   properties {widgetStyle {
     clickedSlateSound {resourceObject=null}
     disabled {
      bIsDynamicallyLoaded=false
      drawAs=RoundedBox
      imageType=NoImage
      mirroring=NoMirror
      resourceName=None
      resourceObject=null
      tiling=NoTile
      imageSize=<0 0>
      margin=<ltrb 0 0 0 0>
      outlineSettings {
       bUseBrushTransparency=false
       roundingType=FixedRadius
       width=1
       color {colorUseRule=UseColor_ColorTable, specifiedColor=<rgba 0 0 0 0>}
       cornerRadii=<4 4 4 4>
      }
      tintColor {colorUseRule=UseColor_Specified, specifiedColor=<rgba 0.0395 0.0395 0.0395 1>}
      uVRegion {bIsValid=false, max=<0 0>, min=<0 0>}
     }
     disabledForeground {colorUseRule=UseColor_ColorTable, specifiedColor=<rgba 0 0 0 0>}
     hovered {
      bIsDynamicallyLoaded=false
      drawAs=RoundedBox
      imageType=NoImage
      mirroring=NoMirror
      resourceName=None
      resourceObject=null
      tiling=NoTile
      imageSize=<32 32>
      margin=<ltrb 0 0 0 0>
      outlineSettings {
       bUseBrushTransparency=false
       roundingType=FixedRadius
       width=1
       color {colorUseRule=UseColor_ColorTable, specifiedColor=<rgba 0 0 0 0>}
       cornerRadii=<4 4 4 4>
      }
      tintColor {colorUseRule=UseColor_Specified, specifiedColor=<rgba 0.0953 0.0953 0.0953 1>}
      uVRegion {bIsValid=false, max=<0 0>, min=<0 0>}
     }
     hoveredForeground {colorUseRule=UseColor_ColorTable, specifiedColor=<rgba 0 0 0 0>}
     hoveredSlateSound {resourceObject=null}
     normal {
      bIsDynamicallyLoaded=false
      drawAs=RoundedBox
      imageType=NoImage
      mirroring=NoMirror
      resourceName=None
      resourceObject=null
      tiling=NoTile
      imageSize=<32 32>
      margin=<ltrb 0 0 0 0>
      outlineSettings {
       bUseBrushTransparency=false
       roundingType=FixedRadius
       width=1
       color {colorUseRule=UseColor_ColorTable, specifiedColor=<rgba 0 0 0 0>}
       cornerRadii=<4 4 4 4>
      }
      tintColor {colorUseRule=UseColor_Specified, specifiedColor=<rgba 0.0395 0.0395 0.0395 1>}
      uVRegion {bIsValid=false, max=<0 0>, min=<0 0>}
     }
     normalForeground {colorUseRule=UseColor_ColorTable, specifiedColor=<rgba 0 0 0 0>}
     normalPadding=<ltrb 12 1.5 12 1.5>
     pressed {
      bIsDynamicallyLoaded=false
      drawAs=RoundedBox
      imageType=NoImage
      mirroring=NoMirror
      resourceName=None
      resourceObject=null
      tiling=NoTile
      imageSize=<32 32>
      margin=<ltrb 0 0 0 0>
      outlineSettings {
       bUseBrushTransparency=false
       roundingType=FixedRadius
       width=1
       color {colorUseRule=UseColor_ColorTable, specifiedColor=<rgba 0 0 0 0>}
       cornerRadii=<4 4 4 4>
      }
      tintColor {colorUseRule=UseColor_Specified, specifiedColor=<rgba 0.0284 0.0284 0.0284 1>}
      uVRegion {bIsValid=false, max=<0 0>, min=<0 0>}
     }
     pressedForeground {colorUseRule=UseColor_ColorTable, specifiedColor=<rgba 0 0 0 0>}
     pressedPadding=<ltrb 12 2.5 12 0.5>
     pressedSlateSound {resourceObject=null}
    }}
   slot {class=VerticalBoxSlot}
  }
  VariantToolButton {
   widgetClassPath=&CommonTextBlock
   parent=UpdateHeadButton
   widget=VariantToolButton
   properties {
    text=Update Head
    font {
     bForceMonospaced=false
     bMaterialIsStencil=false
     fontMaterial=null
     fontObject=&BurbankSmall
     letterSpacing=0
     monospacedWidth=1
     size=17
     skewAmount=0
     typefaceFontName=Medium
     outlineSettings {
      bApplyOutlineToDropShadows=false
      bMiteredCorners=false
      bSeparateFillAlpha=false
      outlineMaterial=null
      outlineSize=0
      outlineColor=<rgba 0 0 0 1>
     }
    }
    shadowOffset=<0 0>
   }
   slot {class=ButtonSlot, properties {padding=<ltrb 10 0 10 0>}}
  }
  UpdateTexturesButton_1 {
   widgetClassPath=&Button
   parent=VerticalBox_0
   widget=UpdateTexturesButton_1
   properties {
    visibility=Hidden
    widgetStyle {
     clickedSlateSound {resourceObject=null}
     disabled {
      bIsDynamicallyLoaded=false
      drawAs=RoundedBox
      imageType=NoImage
      mirroring=NoMirror
      resourceName=None
      resourceObject=null
      tiling=NoTile
      imageSize=<0 0>
      margin=<ltrb 0 0 0 0>
      outlineSettings {
       bUseBrushTransparency=false
       roundingType=FixedRadius
       width=1
       color {colorUseRule=UseColor_ColorTable, specifiedColor=<rgba 0 0 0 0>}
       cornerRadii=<4 4 4 4>
      }
      tintColor {colorUseRule=UseColor_Specified, specifiedColor=<rgba 0.0395 0.0395 0.0395 1>}
      uVRegion {bIsValid=false, max=<0 0>, min=<0 0>}
     }
     disabledForeground {colorUseRule=UseColor_ColorTable, specifiedColor=<rgba 0 0 0 0>}
     hovered {
      bIsDynamicallyLoaded=false
      drawAs=RoundedBox
      imageType=NoImage
      mirroring=NoMirror
      resourceName=None
      resourceObject=null
      tiling=NoTile
      imageSize=<32 32>
      margin=<ltrb 0 0 0 0>
      outlineSettings {
       bUseBrushTransparency=false
       roundingType=FixedRadius
       width=1
       color {colorUseRule=UseColor_ColorTable, specifiedColor=<rgba 0 0 0 0>}
       cornerRadii=<4 4 4 4>
      }
      tintColor {colorUseRule=UseColor_Specified, specifiedColor=<rgba 0.0953 0.0953 0.0953 1>}
      uVRegion {bIsValid=false, max=<0 0>, min=<0 0>}
     }
     hoveredForeground {colorUseRule=UseColor_ColorTable, specifiedColor=<rgba 0 0 0 0>}
     hoveredSlateSound {resourceObject=null}
     normal {
      bIsDynamicallyLoaded=false
      drawAs=RoundedBox
      imageType=NoImage
      mirroring=NoMirror
      resourceName=None
      resourceObject=null
      tiling=NoTile
      imageSize=<32 32>
      margin=<ltrb 0 0 0 0>
      outlineSettings {
       bUseBrushTransparency=false
       roundingType=FixedRadius
       width=1
       color {colorUseRule=UseColor_ColorTable, specifiedColor=<rgba 0 0 0 0>}
       cornerRadii=<4 4 4 4>
      }
      tintColor {colorUseRule=UseColor_Specified, specifiedColor=<rgba 0.0395 0.0395 0.0395 1>}
      uVRegion {bIsValid=false, max=<0 0>, min=<0 0>}
     }
     normalForeground {colorUseRule=UseColor_ColorTable, specifiedColor=<rgba 0 0 0 0>}
     normalPadding=<ltrb 12 1.5 12 1.5>
     pressed {
      bIsDynamicallyLoaded=false
      drawAs=RoundedBox
      imageType=NoImage
      mirroring=NoMirror
      resourceName=None
      resourceObject=null
      tiling=NoTile
      imageSize=<32 32>
      margin=<ltrb 0 0 0 0>
      outlineSettings {
       bUseBrushTransparency=false
       roundingType=FixedRadius
       width=1
       color {colorUseRule=UseColor_ColorTable, specifiedColor=<rgba 0 0 0 0>}
       cornerRadii=<4 4 4 4>
      }
      tintColor {colorUseRule=UseColor_Specified, specifiedColor=<rgba 0.0284 0.0284 0.0284 1>}
      uVRegion {bIsValid=false, max=<0 0>, min=<0 0>}
     }
     pressedForeground {colorUseRule=UseColor_ColorTable, specifiedColor=<rgba 0 0 0 0>}
     pressedPadding=<ltrb 12 2.5 12 0.5>
     pressedSlateSound {resourceObject=null}
    }
   }
   slot {class=VerticalBoxSlot}
  }
  VariantToolButton_1 {
   widgetClassPath=&CommonTextBlock
   parent=UpdateTexturesButton_1
   widget=VariantToolButton_1
   properties {
    text=UpdateTextures
    font {
     bForceMonospaced=false
     bMaterialIsStencil=false
     fontMaterial=null
     fontObject=&BurbankSmall
     letterSpacing=0
     monospacedWidth=1
     size=17
     skewAmount=0
     typefaceFontName=Medium
     outlineSettings {
      bApplyOutlineToDropShadows=false
      bMiteredCorners=false
      bSeparateFillAlpha=false
      outlineMaterial=null
      outlineSize=0
      outlineColor=<rgba 0 0 0 1>
     }
    }
    shadowOffset=<0 0>
   }
   slot {class=ButtonSlot, properties {padding=<ltrb 10 0 10 0>}}
  }
  Render_Detail_VerticalBox {widgetClassPath=&VerticalBox, parent=ScrollBox_0, widget=Render_Detail_VerticalBox, slot {class=ScrollBoxSlot}}
  Render_DetailsView {
   widgetClassPath=&DetailsView
   parent=Render_Detail_VerticalBox
   widget=Render_DetailsView
   properties {categoriesToShow [RenderOptions, RenderList]}
   slot {class=VerticalBoxSlot}
  }
  Render_Button {
   widgetClassPath=&Button
   parent=Render_Detail_VerticalBox
   widget=Render_Button
   properties {widgetStyle {
     clickedSlateSound {resourceObject=null}
     disabled {
      bIsDynamicallyLoaded=false
      drawAs=RoundedBox
      imageType=NoImage
      mirroring=NoMirror
      resourceName=None
      resourceObject=null
      tiling=NoTile
      imageSize=<0 0>
      margin=<ltrb 0 0 0 0>
      outlineSettings {
       bUseBrushTransparency=false
       roundingType=FixedRadius
       width=1
       color {colorUseRule=UseColor_ColorTable, specifiedColor=<rgba 0 0 0 0>}
       cornerRadii=<4 4 4 4>
      }
      tintColor {colorUseRule=UseColor_Specified, specifiedColor=<rgba 0.0395 0.0395 0.0395 1>}
      uVRegion {bIsValid=false, max=<0 0>, min=<0 0>}
     }
     disabledForeground {colorUseRule=UseColor_ColorTable, specifiedColor=<rgba 0 0 0 0>}
     hovered {
      bIsDynamicallyLoaded=false
      drawAs=RoundedBox
      imageType=NoImage
      mirroring=NoMirror
      resourceName=None
      resourceObject=null
      tiling=NoTile
      imageSize=<32 32>
      margin=<ltrb 0 0 0 0>
      outlineSettings {
       bUseBrushTransparency=false
       roundingType=FixedRadius
       width=1
       color {colorUseRule=UseColor_ColorTable, specifiedColor=<rgba 0 0 0 0>}
       cornerRadii=<4 4 4 4>
      }
      tintColor {colorUseRule=UseColor_Specified, specifiedColor=<rgba 0.0953 0.0953 0.0953 1>}
      uVRegion {bIsValid=false, max=<0 0>, min=<0 0>}
     }
     hoveredForeground {colorUseRule=UseColor_ColorTable, specifiedColor=<rgba 0 0 0 0>}
     hoveredSlateSound {resourceObject=null}
     normal {
      bIsDynamicallyLoaded=false
      drawAs=RoundedBox
      imageType=NoImage
      mirroring=NoMirror
      resourceName=None
      resourceObject=null
      tiling=NoTile
      imageSize=<32 32>
      margin=<ltrb 0 0 0 0>
      outlineSettings {
       bUseBrushTransparency=false
       roundingType=FixedRadius
       width=1
       color {colorUseRule=UseColor_ColorTable, specifiedColor=<rgba 0 0 0 0>}
       cornerRadii=<4 4 4 4>
      }
      tintColor {colorUseRule=UseColor_Specified, specifiedColor=<rgba 0.0395 0.0395 0.0395 1>}
      uVRegion {bIsValid=false, max=<0 0>, min=<0 0>}
     }
     normalForeground {colorUseRule=UseColor_ColorTable, specifiedColor=<rgba 0 0 0 0>}
     normalPadding=<ltrb 12 1.5 12 1.5>
     pressed {
      bIsDynamicallyLoaded=false
      drawAs=RoundedBox
      imageType=NoImage
      mirroring=NoMirror
      resourceName=None
      resourceObject=null
      tiling=NoTile
      imageSize=<32 32>
      margin=<ltrb 0 0 0 0>
      outlineSettings {
       bUseBrushTransparency=false
       roundingType=FixedRadius
       width=1
       color {colorUseRule=UseColor_ColorTable, specifiedColor=<rgba 0 0 0 0>}
       cornerRadii=<4 4 4 4>
      }
      tintColor {colorUseRule=UseColor_Specified, specifiedColor=<rgba 0.0284 0.0284 0.0284 1>}
      uVRegion {bIsValid=false, max=<0 0>, min=<0 0>}
     }
     pressedForeground {colorUseRule=UseColor_ColorTable, specifiedColor=<rgba 0 0 0 0>}
     pressedPadding=<ltrb 12 2.5 12 0.5>
     pressedSlateSound {resourceObject=null}
    }}
   slot {class=VerticalBoxSlot}
  }
  Render_CommonText {
   widgetClassPath=&CommonTextBlock
   parent=Render_Button
   widget=Render_CommonText
   properties {
    text=Render
    font {
     bForceMonospaced=false
     bMaterialIsStencil=false
     fontMaterial=null
     fontObject=&BurbankSmall
     letterSpacing=0
     monospacedWidth=1
     size=17
     skewAmount=0
     typefaceFontName=Medium
     outlineSettings {
      bApplyOutlineToDropShadows=false
      bMiteredCorners=false
      bSeparateFillAlpha=false
      outlineMaterial=null
      outlineSize=0
      outlineColor=<rgba 0 0 0 1>
     }
    }
    shadowOffset=<0 0>
   }
   slot {class=ButtonSlot, properties {padding=<ltrb 10 0 10 0>}}
  }
 ]
}
